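# Translucent UI material that draws curved cooldown brackets with a progress fill and drop shadow.
# Bracket, shadow and progress masks intersect MF_UI_Circle SDF outlines with MF_UI_RadialSegments wedges, smoothstepped and shared through named reroutes.
# Progress sets the progress wedge spacing, and IsCooldownInProgress blends the fill toward the cooldown color.
# Masks weight fill, unfilled and shadow colors, with separate alpha parameters scaled by Cooldown_Alpha; HighlightColor is unconnected.
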Material /ColdDayUI/Materials/M_UI_ColdDay_Cooldown_Brackets
graph {
 properties {blendMode=BLEND_Translucent, materialDomain=MD_UI, pixelDepthOffsetMode=PDOM_Legacy, usedShadingModels=Default Lit}
 nodes [
  n1 = NamedRerouteUsage()
  n2 = NamedRerouteUsage()
  n3 = LinearInterpolate(
   A=VectorParameter() {parameterName=CooldownBracketsFillColor, defaultValue=<rgba 0.8953 0.2611 1 1>}.RGB
   B=VectorParameter() {parameterName=CooldownColor, defaultValue=<rgba 0.1484 0.1484 0.1484 1>}.RGB
   Alpha=ScalarParameter() {parameterName=IsCooldownInProgress}
  )
  n4 = Subtract(A=NamedRerouteUsage(), B=NamedRerouteUsage())
  n5 = VectorParameter() {parameterName=CooldownBracketsUnfilledColor, defaultValue=<rgba 0.04 0.04 0.04 1>}
  n6 = NamedRerouteUsage()
  Add(
   A=Multiply(A=VectorParameter() {parameterName=CooldownBracketsShadow, defaultValue=<rgba 0 0 0 1>}.RGB, B=Subtract(A=@n1, B=@n2))
   B=Add(A=Multiply(A=@n3, B=@n4), B=Multiply(A=@n5.RGB, B=@n6))
  )
  n7 = Subtract(A=NamedRerouteUsage(), B=OneMinus(None=NamedRerouteUsage()))
  n8 = ScalarParameter() {defaultValue=1, parameterName=ProgressUnfilledAlpha}
  n9 = Subtract(
   A=LinearInterpolate(A=NamedRerouteUsage(), B=NamedRerouteUsage(), Alpha=ScalarParameter() {defaultValue=0.25, parameterName=ShadowOpacity})
   B=NamedRerouteUsage()
  )
  n10 = ScalarParameter() {defaultValue=1, parameterName=ProgressFillAlpha}
  Multiply(A=Add(A=Multiply(A=@n7, B=@n8), B=Multiply(A=@n9, B=@n10)), B=ScalarParameter() {defaultValue=0.8, parameterName=Cooldown_Alpha})
  n11 = Constant() {r=0.01}
  n12 = ScalarParameter() {defaultValue=0.018, parameterName=ShadowThickness}
  n13 = SmoothStep(Value=MaterialFunctionCall(
    UVs=NamedRerouteUsage()
    Size=ScalarParameter() {defaultValue=0.86, parameterName=BracketsSize}
    OutlineThickness=Add(A=@n11, B=@n12) {constB=0.015}
   ) {materialFunction=&MF_UI_Circle}.SDFOutline) {constMax=0.025, constMin=0.001}
  n14 = SmoothStep(Value=MaterialFunctionCall!(UV, Segment Count=Constant() {r=2}, Segment Spacing=ScalarParameter() {defaultValue=0.24, parameterName=BracketsAngle}, Fixed Spacing) {materialFunction=/FortUILibrary/MaterialFunctions/Basics/MF_UI_RadialSegments}.Segmented Triangles Gradient) {constMax=0.05, constMin=0.001}
  NamedRerouteDeclaration(None=Multiply(A=OneMinus(None=@n13), B=OneMinus(None=@n14))) {name=CooldownBracketsShadow, nodeColor=<rgba 0 0 0 1>}
  n15 = NamedRerouteUsage()
  n16 = ScalarParameter() {defaultValue=0.86, parameterName=BracketsSize}
  n17 = Constant() {r=0.01}
  n18 = OneMinus(None=SmoothStep(Value=MaterialFunctionCall(UVs=@n15, Size=@n16, OutlineThickness=@n17) {materialFunction=&MF_UI_Circle}.SDFOutline) {constMax=0.02})
  n19 = Constant() {r=1}
  n20 = LinearInterpolate(Alpha=ScalarParameter() {defaultValue=1, parameterName=Progress}) {constA=0.37, constB=0.62}
  n21 = OneMinus(None=SmoothStep(Value=MaterialFunctionCall(Segment Count=@n19, Segment Spacing=@n20) {materialFunction=&MF_UI_RadialSegments}.Segmented Triangles Gradient) {constMax=0.01})
  NamedRerouteDeclaration(None=Subtract(A=NamedRerouteUsage(), B=Multiply(A=@n18, B=@n21))) {name=CooldownProgressArch, nodeColor=<rgba 1 0.8438 0 1>}
  NamedRerouteDeclaration(None=TextureCoordinate()) {name=UV, nodeColor=<rgba 0 0 0 1>}
  n22 = SmoothStep(Value=MaterialFunctionCall(UVs=NamedRerouteUsage(), Size=ScalarParameter() {defaultValue=0.86, parameterName=BracketsSize}, OutlineThickness=Constant() {r=0.01}) {materialFunction=&MF_UI_Circle}.SDFOutline) {constMax=0.02}
  n23 = SmoothStep(Value=MaterialFunctionCall(Segment Count=Constant() {r=2}, Segment Spacing=ScalarParameter() {defaultValue=0.24, parameterName=BracketsAngle}) {materialFunction=&MF_UI_RadialSegments}.Segmented Triangles Gradient) {constMax=0.01}
  NamedRerouteDeclaration(None=Multiply(A=OneMinus(None=@n22), B=OneMinus(None=@n23))) {name=CooldownArches, nodeColor=<rgba 0.9453 1 0 1>}
  VectorParameter() {parameterName=HighlightColor, defaultValue=<rgba 1 1 1 1>}
 ]
}
