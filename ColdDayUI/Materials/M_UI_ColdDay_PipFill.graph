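# Translucent UI material for a ColdDay pip fill, apparently on a reticle, showing percentage progress with cooldown and ready states.
# A channel-packed distance field from the reticle texture gives the fill and soft shadow, while a rotatable linear gradient builds the progress mask.
# Parameters blend base, cooldown and ready colors and switch in a box-SDF glow.
# A swirl texture times a radial gradient drives an animatable reveal mask.
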Material /ColdDayUI/Materials/M_UI_ColdDay_PipFill
graph {
 properties {blendMode=BLEND_Translucent, materialDomain=MD_UI, pixelDepthOffsetMode=PDOM_Legacy, usedShadingModels=Default Lit}
 nodes [
  n1 = VectorParameter() {parameterName=ShadowColor, defaultValue=<rgba 0.8438 0.8438 0.8438 1>}
  n2 = OneMinus(None=NamedRerouteUsage())
  n3 = VectorParameter() {parameterName=ReloadingColor, defaultValue=<rgba 0.0781 0.0781 0.0781 0.6>}
  n4 = OneMinus(None=NamedRerouteUsage())
  n5 = Multiply(
   A=LinearInterpolate(
    A=VectorParameter() {parameterName=ReloadedColor, defaultValue=<rgba 0.4125 0.6105 0.0865 1>}.RGB
    B=VectorParameter() {parameterName=HighlightColor, defaultValue=<rgba 1 1 1 1>}.RGB
    Alpha=ScalarParameter() {parameterName=HighlightAlpha}
   )
   B=NamedRerouteUsage()
  )
  n6 = LinearInterpolate(B=OneMinus(None=NamedRerouteUsage()), Alpha=ScalarParameter() {defaultValue=0.7, parameterName=ShadowOpacity}) {constB=0}
  Add(
   A=LinearInterpolate(B=Multiply(A=@n1.RGB, B=@n2), Alpha=ScalarParameter() {defaultValue=0.7, parameterName=ShadowOpacity}) {constB=0}
   B=Add(A=Multiply(A=@n3.RGB, B=@n4), B=Subtract(A=@n5, B=@n6))
  )
  Frac(None=Multiply(A=Time()) {constB=0.2})
  n7 = LinearInterpolate(
   A=Constant3Vector() {constant=<rgba 0 0 0 1>}
   B=VectorParameter() {parameterName=PipBaseColor, defaultValue=<rgba 0.0156 0.0156 0.0156 1>}.RGB
   Alpha=NamedRerouteUsage()
  )
  n8 = VectorParameter() {parameterName=CooldownColor, defaultValue=<rgba 0.149 0.149 0.149 1>}
  n9 = NamedRerouteUsage()
  n10 = VectorParameter() {parameterName=ReadyColor, defaultValue=<rgba 0.4125 0.6105 0.0865 1>}
  n11 = NamedRerouteUsage()
  n12 = NamedRerouteUsage()
  n13 = NamedRerouteUsage()
  LinearInterpolate(
   A=LinearInterpolate(
    A=LinearInterpolate(A=@n7, B=@n8.RGB, Alpha=@n9)
    B=LinearInterpolate(A=@n7, B=@n10.RGB, Alpha=@n11)
    Alpha=ScalarParameter() {parameterName=IsPipReadyforUse}
   )
   B=Multiply(A=VectorParameter() {group=Anim_Glow, parameterName=GlowColor, defaultValue=<rgba 0.8336 1 0.4062 1>}.RGB, B=Add(A=@n12, B=@n13))
   Alpha=ScalarParameter() {group=Anim_Glow, parameterName=GlowSwitch}
  )
  n14 = ChannelMaskParameter(None=TextureSampleParameter2D() {parameterName=CanWallTexture, texture=/ColdDayUI/Textures/T_UI_Reticle_ColdDay}.G) {parameterName=ChannelMask}
  MaterialFunctionCall!(RGBA Channels=@n14.R, Glow Min=Constant(), Glow Max=Constant() {r=1.2623}, Outline Thickness) {materialFunction=/FortUILibrary/MaterialFunctions/Basics/MF_UI_ChannelPackedDistanceField}
  n15 = NamedRerouteUsage()
  n16 = NamedRerouteUsage()
  Multiply(
   A=LinearInterpolate(A=NamedRerouteUsage(), B=Add(A=@n15, B=@n16), Alpha=ScalarParameter() {group=Anim_Glow, parameterName=GlowSwitch})
   B=NamedRerouteUsage()
  )
  n17 = MaterialFunctionCall(
   RGBA Channels=@n14.R
   Glow Min=ScalarParameter() {defaultValue=0.65, group=SDFConfig, parameterName=GlowMin}
   Glow Max=ScalarParameter() {defaultValue=0.8, group=SDFConfig, parameterName=GlowMax}
  ) {materialFunction=&MF_UI_ChannelPackedDistanceField}
  n18 = NamedRerouteDeclaration(None=Multiply(A=@n17.Fill) {constB=0.9}) {name=Opacity, nodeColor=<rgba 0 0 0 1>}
  n19 = Subtract(
   A=SmoothStep(Value=MaterialFunctionCall(RGBA Channels=@n14.R) {materialFunction=&MF_UI_ChannelPackedDistanceField}.DF) {constMin=0.1}
   B=@n17.Fill
  )
  n20 = ScalarParameter() {defaultValue=0.7, parameterName=ShadowOpacity}
  NamedRerouteDeclaration(None=Add(A=@n18, B=Multiply(A=@n19, B=@n20))) {name=Opacity With Shadow, nodeColor=<rgba 0 0 0 1>}
  n21 = Multiply(
   A=TextureSampleParameter2D() {parameterName=FX, texture=/ColdDayUI/Textures/T_UI_ColdDay_SwirlFlow}.RGB
   B=MaterialFunctionCall(Radius=Constant() {r=0.4}, Density=Constant() {r=2}) {materialFunction=&RadialGradientExponential}.RadialGradientExponential
  )
  n22 = ScalarParameter() {group=Anim_Reveal, parameterName=RevealControl}
  NamedRerouteDeclaration(None=OneMinus(None=SmoothStep(Max=@n21, Value=@n22))) {name=PipFillRevealMask, nodeColor=<rgba 0.4062 1 0 1>}
  n23 = ScalarParameter() {defaultValue=1, parameterName=Percentage}
  n24 = ScalarParameter() {defaultValue=0.05, group=PercentageConfig, parameterName=MinPercentage}
  n25 = ScalarParameter() {defaultValue=0.92, group=PercentageConfig, parameterName=MaxPercentage}
  n26 = Constant()
  n27 = MaterialFunctionCall(
   GradientStart=Constant()
   GradientEnd=Constant() {r=1}
   RotationAngle=ScalarParameter() {defaultValue=0.25, parameterName=RotationAngle}
   GradientOffset=If(A=@n23, "A > B"=LinearInterpolate(A=@n24, B=@n25, Alpha=@n23) {constA=0.05, constB=0.92}, "A == B"=@n26, "A < B"=@n26)
  ) {materialFunction=&MF_UI_LinearGradient}
  NamedRerouteDeclaration(None=OneMinus(None=SmoothStep(Value=@n27.Result) {constMax=0.01})) {name=Progress, nodeColor=<rgba 0 0 0 1>}
  n28 = MaterialFunctionCall(
   Size=AppendVector(
    A=ScalarParameter() {defaultValue=-0.08, group=Anim_Glow, parameterName=GlowXScale}
    B=ScalarParameter() {defaultValue=-0.2, group=Anim_Glow, parameterName=GlowYScale}
   )
   GlowMin=Constant() {r=0.1}
   GlowMax=Constant() {r=0.6}
  ) {materialFunction=&MF_UI_Box}
  NamedRerouteDeclaration(None=Multiply(A=@n28.Fill, B=ScalarParameter() {defaultValue=1, group=Anim_Glow, parameterName=GlowOpacity})) {name=Glow, nodeColor=<rgba 0 0 0 1>}
  NamedRerouteDeclaration(None=Multiply(A=@n18, B=NamedRerouteUsage())) {name=ProgressOpacity, nodeColor=<rgba 0.0547 1 0 1>}
  NamedRerouteDeclaration(None=@n28.SDF) {nodeColor=<rgba 0 0.8125 1 1>}
 ]
}
