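# Translucent UI material for the ColdDay bending reticle: outer brackets, an element-specific center and an optional dot.
# Mirrored bracket and channel-masked center textures (fire, earth, water, air) pass through distance-field functions that derive fills and soft shadows.
# A static switch handles the Firebending center, built from mirrored pieces with animation or in-range parameters.
# Panning swirl noise can modulate center opacity, and IsTargetInRange blends the dot layer.
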Material /ColdDayUI/Materials/M_UI_ColdDay_Reticle
graph {
 properties {blendMode=BLEND_Translucent, materialDomain=MD_UI, pixelDepthOffsetMode=PDOM_Legacy, usedShadingModels=Default Lit}
 nodes [
  Frac(None=Time())
  n1 = Constant()
  n2 = ScalarParameter() {group=MiddleOfReticleConfig, parameterName=MiddleOfReticleYPadding}
  n3 = MaterialFunctionCall(
   UVs=Add(A=AppendVector(A=@n1, B=@n2), B=NamedRerouteUsage())
   Texture Scale=ScalarParameter() {defaultValue=0.25, group=MiddleOfReticleConfig, parameterName=MiddleOfReticleScale}
  ) {materialFunction=&ScaleUVsByCenter}
  n4 = ScalarParameter() {parameterName=CenterRotation}
  n5 = ChannelMaskParameter(None=TextureSample(UVs=Rotator(Coordinate=@n3.UVs, Time=@n4)) {samplerSource=SSM_Clamp_WorldGroupSettings, texture=/ColdDayUI/Textures/T_UI_Reticle_Center_ColdDay}.RGBA) {
   desc="R = Firebending\r\nG = Earthbending\r\nB = Waterbending\r\nA = Airbending"
   group=MiddleOfReticleConfig
   maskChannel=Blue
   parameterName=UseWhichCenterOfReticle
   defaultValue=<rgba 0 0 1 0>
  }
  n6 = MaterialFunctionCall!(
   RGBA Channels=@n5.R
   Glow Min=ScalarParameter() {defaultValue=0.386, group=MiddleOfReticleConfig, parameterName=GlowMin}
   Glow Max=ScalarParameter() {defaultValue=0.5, group=MiddleOfReticleConfig, parameterName=GlowMax}
   Outline Thickness
  ) {materialFunction=/FortUILibrary/MaterialFunctions/Basics/MF_UI_ChannelPackedDistanceField}
  NamedRerouteDeclaration(None=LinearInterpolate(A=SmoothStep(Value=@n6.Fill) {constMax=0.5, constMin=0.002}, B=@n6.Fill, Alpha=@n6.Fill)) {name=ReticleCenterShadow, nodeColor=<rgba 0 0 0 1>}
  n7 = Constant2Vector() {r=0.35}
  n8 = Constant2Vector() {g=0.45, r=1}
  n9 = MaterialFunctionCall(
   RGBA Channels=TextureSample(UVs=MaterialFunctionCall!(UV, Offset=@n7, Rotation, Scale=@n8, MirrorY) {materialFunction=/FortUILibrary/MaterialFunctions/Basics/MF_UI_Symmetry_2x1_DEPRECATED}.Result) {texture=/ColdDayUI/Textures/T_UI_Reticle_ColdDay}.R
   Glow Min=Constant() {r=0.25}
   Glow Max=Constant() {r=0.5}
  ) {materialFunction=&MF_UI_ChannelPackedDistanceField}
  n10 = NamedRerouteDeclaration(None=@n9.Fill) {name=MainBrackets, nodeColor=<rgba 0 0 0 1>}
  NamedRerouteDeclaration(None=LinearInterpolate(A=@n10, B=SmoothStep(Value=@n9.Fill) {constMax=0.03, constMin=0.0001}, Alpha=@n10) {constAlpha=0.25}) {name=MainBracketsShadow, nodeColor=<rgba 0 0 0 1>}
  NamedRerouteDeclaration(None=@n6.Fill) {name=ReticleCenter, nodeColor=<rgba 0 0 0 1>}
  n11 = MaterialFunctionCall(UVs=NamedRerouteUsage(), Size=ScalarParameter() {defaultValue=0.005, group=CenterDotConfig, parameterName=DotScale}) {materialFunction=&MF_UI_Circle}
  n12 = OneMinus(None=SmoothStep(Value=@n11.SDF) {constMax=0.04, constMin=0.02})
  n13 = NamedRerouteDeclaration(None=OneMinus(None=SmoothStep(Value=@n11.SDF) {constMax=0.02, constMin=0.01})) {name=Dot, nodeColor=<rgba 0 0 0 1>}
  NamedRerouteDeclaration(None=LinearInterpolate(A=Multiply(A=@n12) {constB=0.5}, B=@n13, Alpha=@n13)) {name=DotShadow, nodeColor=<rgba 0 0 0 1>}
  NamedRerouteDeclaration(None=TextureCoordinate()) {name=UV, nodeColor=<rgba 0 0 0 1>}
  n14 = Panner(Coordinate=TextureCoordinate() {uTiling=0.5, vTiling=0.5}, Speed=AppendVector(A=Constant() {r=0.2}, B=Constant() {r=0.4097})) {speedY=0.5}
  n15 = Panner(Speed=AppendVector(A=Constant() {r=-0.15}, B=Constant() {r=0.388})) {speedY=0.2}
  n16 = Multiply(B=Add(
    A=TextureSampleParameter2D(UVs=@n14) {parameterName=Terxture_01, texture=/ColdDayUI/Textures/T_UI_ColdDay_SwirlFlow}.G
    B=TextureSampleParameter2D(UVs=@n15) {parameterName=Texure_02, texture=&T_UI_ColdDay_SwirlFlow}.G
   )) {constA=1}
  NamedRerouteDeclaration(None=SmoothStep(Value=Saturate(None=@n16))) {name=WaterNoise, nodeColor=<rgba 1 0 0.6021 1>}
  n17 = ScalarParameter() {defaultValue=-0.318, group=FireBendingConfig, parameterName=FireBendingCenterOffset}
  n18 = Constant()
  n19 = Constant2Vector() {g=0.125, r=0.25}
  n20 = ScalarParameter() {defaultValue=1.7, group=FireBendingConfig, parameterName=FireBendingSymmetryCenterScale}
  n21 = TextureSample(UVs=MaterialFunctionCall(Offset=AppendVector(A=@n17, B=@n18), Scale=Multiply(A=@n19, B=@n20)) {materialFunction=&MF_UI_Symmetry_2x1_DEPRECATED}.Result) {samplerSource=SSM_Clamp_WorldGroupSettings, texture=/ColdDayUI/Textures/T_UI_Reticle_Center_AnnoyedMop}
  n22 = Constant() {r=-0.02}
  n23 = MaterialFunctionCall!(
   UVs=MaterialFunctionCall(Offset=AppendVector(A=@n17, B=@n22), Scale=Constant() {r=0.5}) {materialFunction=&MF_UI_Symmetry_2x1_DEPRECATED}.Result
   Offset
   RotationCenter
   RotationAngle=Constant() {r=0.202}
  ) {materialFunction=/FortUILibrary/MaterialFunctions/Gradients/MF_UI_ReflectedGradient}
  n24 = Add(
   A=OneMinus(None=SmoothStep(Value=@n23.Result) {constMin=0.7})
   B=ScalarParameter() {group=FireBendingConfig, parameterName=FireBendingCenter_Anim}
  )
  n25 = MaterialFunctionCall() {materialFunction=&MF_UI_ReflectedGradient}
  n26 = Add(
   A=OneMinus(None=SmoothStep(Value=@n25.Result) {constMin=0.3})
   B=ScalarParameter() {group=FireBendingConfig, parameterName=FireBendingCenter_Anim}
  )
  n27 = Multiply(
   A=MaterialFunctionCall(
    RGBA Channels=Add(A=@n21.R, B=@n21.R)
    Glow Min=ScalarParameter() {defaultValue=0.386, group=MiddleOfReticleConfig, parameterName=GlowMin}
    Glow Max=ScalarParameter() {defaultValue=0.5, group=MiddleOfReticleConfig, parameterName=GlowMax}
   ) {materialFunction=&MF_UI_ChannelPackedDistanceField}.Fill
   B=Multiply(A=SmoothStep(Value=@n24) {constMax=0.55, constMin=0.5}, B=SmoothStep(Value=@n26) {constMax=0.16, constMin=0.15})
  )
  NamedRerouteDeclaration(None=@n27) {name=FireBendingCenter, nodeColor=<rgba 1 0 0 1>}
  NamedRerouteDeclaration(None=LinearInterpolate(A=SmoothStep(Value=@n27) {constMax=0.5, constMin=0.002}, B=@n27, Alpha=@n27)) {name=FireBendingCenterShadow, nodeColor=<rgba 1 0 0 1>}
  n28 = AppendVector(A=Constant() {r=-0.33}, B=Constant())
  n29 = AppendVector(A=Constant() {r=-0.414}, B=Constant() {r=0.056})
  n30 = ScalarParameter() {parameterName=FirebendingInRange}
  n31 = Constant2Vector() {g=0.125, r=0.25}
  n32 = LinearInterpolate(A=Constant() {r=1.7}, B=Constant() {r=1.2}, Alpha=ScalarParameter() {parameterName=FirebendingInRange})
  n33 = TextureSample(UVs=MaterialFunctionCall(Offset=LinearInterpolate(A=@n28, B=@n29, Alpha=@n30) {constA=-0.37, constB=0.096}, Scale=Multiply(A=@n31, B=@n32)) {materialFunction=&MF_UI_Symmetry_2x1_DEPRECATED}.Result) {samplerSource=SSM_Clamp_WorldGroupSettings, texture=&T_UI_Reticle_Center_AnnoyedMop}
  n34 = AppendVector(A=Constant() {r=-0.314}, B=Constant() {r=0.012})
  n35 = AppendVector(A=Constant() {r=-0.41}, B=Constant() {r=-0.0319})
  n36 = ScalarParameter() {parameterName=FirebendingInRange}
  n37 = Constant2Vector() {g=0.125, r=0.25}
  n38 = Constant() {r=1.7}
  n39 = TextureSample(UVs=MaterialFunctionCall(Offset=LinearInterpolate(A=@n34, B=@n35, Alpha=@n36) {constA=-0.37, constB=0.096}, Scale=Multiply(A=@n37, B=@n38)) {materialFunction=&MF_UI_Symmetry_2x1_DEPRECATED}.Result) {samplerSource=SSM_Clamp_WorldGroupSettings, texture=&T_UI_Reticle_Center_AnnoyedMop}
  n40 = Multiply(A=MaterialFunctionCall(
    RGBA Channels=Add(A=@n33.R, B=@n39.G)
    Glow Min=ScalarParameter() {defaultValue=0.386, group=MiddleOfReticleConfig, parameterName=GlowMin}
    Glow Max=ScalarParameter() {defaultValue=0.5, group=MiddleOfReticleConfig, parameterName=GlowMax}
   ) {materialFunction=&MF_UI_ChannelPackedDistanceField}.Fill)
  NamedRerouteDeclaration(None=@n40) {name=FireBendingCenter_V2, nodeColor=<rgba 1 0 0 1>}
  NamedRerouteDeclaration(None=LinearInterpolate(A=SmoothStep(Value=@n40) {constMax=0.5, constMin=0.002}, B=@n40, Alpha=@n40)) {name=FireBendingCenterShadow_V2, nodeColor=<rgba 1 0 0 1>}
  n41 = NamedRerouteUsage()
  n42 = NamedRerouteUsage()
  n43 = StaticSwitchParameter(True=Add(A=NamedRerouteUsage(), B=NamedRerouteUsage()), False=Add(A=NamedRerouteUsage(), B=NamedRerouteUsage())) {group=FireBendingConfig, parameterName=IsFireBendingCenter}
  n44 = Multiply(A=@n43, B=Multiply(A=NamedRerouteUsage(), B=Constant() {r=1}))
  n45 = ScalarParameter() {parameterName=IsNoiseOn}
  n46 = Add(
   A=Multiply(A=Add(A=@n41, B=@n42), B=ScalarParameter() {defaultValue=0.15, parameterName=OuterBracketsAlpha})
   B=Multiply(A=LinearInterpolate(A=@n43, B=@n44, Alpha=@n45), B=ScalarParameter() {defaultValue=0.3, parameterName=CenterOpacity})
  )
  n47 = LinearInterpolate(A=NamedRerouteUsage(), Alpha=ScalarParameter() {group=TargetingArrowSettings, parameterName=IsTargetInRange})
  Reroute(None=StaticSwitchParameter(True=Add(A=@n46, B=@n47), False=@n46) {group=CenterDotConfig, parameterName=EnableCenterDot})
  n48 = Multiply(
   A=VectorParameter() {group=CenterDotConfig, parameterName=CenterDotColor, defaultValue=<rgba 1 1 1 1>}.RGB
   B=LinearInterpolate(A=NamedRerouteUsage(), B=NamedRerouteUsage(), Alpha=ScalarParameter() {group=TargetingArrowSettings, parameterName=IsTargetInRange})
  )
  n49 = Multiply(
   A=VectorParameter() {group=CenterDotConfig, parameterName=CenterDotShadowColor, defaultValue=<rgba 0 0 0 1>}.RGB
   B=LinearInterpolate(A=NamedRerouteUsage(), B=NamedRerouteUsage(), Alpha=ScalarParameter() {group=TargetingArrowSettings, parameterName=IsTargetInRange})
  )
  n50 = NamedRerouteUsage()
  n51 = NamedRerouteUsage()
  n52 = Add(
   A=Multiply(
    A=VectorParameter() {group=MainBracketsConfig, parameterName=ReticleCenterColor, defaultValue=<rgba 1 1 1 1>}.RGB
    B=StaticSwitchParameter(True=@n50, False=@n51) {group=FireBendingConfig, parameterName=IsFireBendingCenter}
   )
   B=Multiply(A=VectorParameter() {group=MainBracketsConfig, parameterName=MainBracketsColor, defaultValue=<rgba 1 1 1 1>}.RGB, B=NamedRerouteUsage())
  )
  StaticSwitchParameter(True=Add(A=Add(A=@n48, B=@n49), B=@n52), False=@n52) {group=CenterDotConfig, parameterName=EnableCenterDot}
 ]
}
